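# Conversation database for the Funding Station NPC, where players choose which war-effort option to fund.
# An entry point leads to a contextual pitch, with an alternate greeting for multi-choice NPCs, which branches into six Funding tasks.
# Each task requires its own Fund1–6 service tag, offers a focus-setting choice and plays service-used effects.
# Tasks map in pairs onto funding choices A, B and C, and all lead to one shared Repeat node.
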ConversationDatabase /CorruptionGameplay/Conversations/Corruption_Conversation_FundingStation
entry_points [{tag=Conversation.Entry.FundingStation, destinations [#1]}]
nodes [
 ConversationEntryPointNode_7 {
  id=#1
  class=ConversationEntryPointNode
  outputs [#2]
  properties {entryTag {tagName=Conversation.Entry.FundingStation}}
 }
 Speech_C_3 {
  id=#2
  class=Speech_C
  outputs [
   #3
   #4
   #5
   #6
   #7
   #8
  ]
  properties {
   bUsePerNPC_ContextualMessageLookup=true
   nodeName=""
   generalConfig {
    defaultMessage="Go on user... you know you want to! Its a worthy cause! "
    contextualMessages [{
      message="Hi there user! We need YOU! What choice'll it be?"
      contextRequirements [{
        participantId {tagName=Conversation.Participant.NPC}
        requirement {class=FortControllerRequirement_NPCHasServiceProviderTag, properties {queryTag {tagName=Conversation.Entry.FundingStation.MultiChoice}}}
       }]
      priority {value=0, curve {curveTable=null, rowName=None}}
     }]
   }
  }
 }
 Funding_C_5 {
  id=#8
  class=Funding_C
  outputs [#9]
  properties {serviceBriefCollectionKeyOveride=Service.Funding.ChoiceA}
  sub_nodes [
   FortConversationRequirement_HasService_0 {class=FortConversationRequirement_HasService, properties {serviceTag {tagName=Tandem.Service.Fund6}}}
   ConversationChoiceNode_3 {
    class=ConversationChoiceNode
    properties {
     defaultChoiceDisplayText="If we wanna be  <WarEffort>top banana</> in this war, I suggest investin' in some <hot>mechanical muscle</>."
     choiceTags {gameplayTags [{tagName=Conversation.Choice.SetFocusUsingCurrentIndex}]}
    }
   }
   SideEffect_PlayServiceUsedEffects_C_1 {class=SideEffect_PlayServiceUsedEffects_C}
  ]
 }
 Repeat_C_0 {id=#9, class=Repeat_C}
 Funding_C_4 {
  id=#7
  class=Funding_C
  outputs [#9]
  properties {fundChoiceIndex=2, serviceBriefCollectionKeyOveride=Service.Funding.ChoiceC}
  sub_nodes [
   FortConversationRequirement_HasService_4 {class=FortConversationRequirement_HasService, properties {serviceTag {tagName=Tandem.Service.Fund5}}}
   ConversationChoiceNode_10 {
    class=ConversationChoiceNode
    properties {
     defaultChoiceDisplayText="Ah. Nice choice to <hot>unvault</>. I love that classic seafaring pirate design."
     choiceTags {gameplayTags [{tagName=Conversation.Choice.SetFocusUsingCurrentIndex}]}
    }
   }
   SideEffect_PlayServiceUsedEffects_C_6 {class=SideEffect_PlayServiceUsedEffects_C}
  ]
 }
 Funding_C_2 {
  id=#6
  class=Funding_C
  outputs [#9]
  properties {fundChoiceIndex=2, serviceBriefCollectionKeyOveride=Service.Funding.ChoiceC}
  sub_nodes [
   FortConversationRequirement_HasService_6 {class=FortConversationRequirement_HasService, properties {serviceTag {tagName=Tandem.Service.Fund3}}}
   ConversationChoiceNode_14 {
    class=ConversationChoiceNode
    properties {
     defaultChoiceDisplayText="Been <WarEffort>monkeyin'</> around in the workshop but I can't decide what to work on."
     choiceTags {gameplayTags [{tagName=Conversation.Choice.SetFocusUsingCurrentIndex}]}
    }
   }
   SideEffect_PlayServiceUsedEffects_C_8 {class=SideEffect_PlayServiceUsedEffects_C}
  ]
 }
 Funding_C_3 {
  id=#5
  class=Funding_C
  outputs [#9]
  properties {fundChoiceIndex=1, serviceBriefCollectionKeyOveride=Service.Funding.ChoiceB}
  sub_nodes [
   FortConversationRequirement_HasService_2 {class=FortConversationRequirement_HasService, properties {serviceTag {tagName=Tandem.Service.Fund4}}}
   ConversationChoiceNode_9 {
    class=ConversationChoiceNode
    properties {
     defaultChoiceDisplayText="<hot>Unvaulting</> this will help keep your friends close and your <WarEffort>bananenemies</> closer."
     choiceTags {gameplayTags [{tagName=Conversation.Choice.SetFocusUsingCurrentIndex}]}
    }
   }
   SideEffect_PlayServiceUsedEffects_C_5 {class=SideEffect_PlayServiceUsedEffects_C}
  ]
 }
 Funding_C_1 {
  id=#4
  class=Funding_C
  outputs [#9]
  properties {fundChoiceIndex=1, serviceBriefCollectionKeyOveride=Service.Funding.ChoiceB}
  sub_nodes [
   FortConversationRequirement_HasService_5 {class=FortConversationRequirement_HasService, properties {serviceTag {tagName=Tandem.Service.Fund2}}}
   ConversationChoiceNode_13 {
    class=ConversationChoiceNode
    properties {
     defaultChoiceDisplayText="Been <WarEffort>monkeyin'</> around in the workshop but I can't decide what to work on."
     choiceTags {gameplayTags [{tagName=Conversation.Choice.SetFocusUsingCurrentIndex}]}
    }
   }
   SideEffect_PlayServiceUsedEffects_C_4 {class=SideEffect_PlayServiceUsedEffects_C}
  ]
 }
 Funding_C_0 {
  id=#3
  class=Funding_C
  outputs [#9]
  properties {serviceBriefCollectionKeyOveride=Service.Funding.ChoiceA}
  sub_nodes [
   FortConversationRequirement_HasService_3 {class=FortConversationRequirement_HasService, properties {serviceTag {tagName=Tandem.Service.Fund1}}}
   ConversationChoiceNode_11 {
    class=ConversationChoiceNode
    properties {
     defaultChoiceDisplayText="You wanna get those <hot>towers</> built, yer gonna have to <WarEffort>chimp in</>."
     choiceTags {gameplayTags [{tagName=Conversation.Choice.SetFocusUsingCurrentIndex}]}
    }
   }
   SideEffect_PlayServiceUsedEffects_C_3 {class=SideEffect_PlayServiceUsedEffects_C}
  ]
 }
]
